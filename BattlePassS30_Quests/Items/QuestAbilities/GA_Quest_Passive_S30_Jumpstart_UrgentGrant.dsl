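; Passive quest gameplay ability for Battle Pass S30 "Jumpstart" that grants an urgent transient quest to the owning player.
; On activation it calls the parent, gets the owning pawn's transient quests component and grants the quest definition held in its "Quest Def Athena to Grant" variable, then ends immediately.
; A failed passive activation only defers to the parent, and the end-ability event is empty.
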(event Ability|EventFailedToActivatePassiveAbility
  (params (ActorInfo :type Gameplay%20Ability%20Actor%20Info%20Structure))
  (|CallParentFunction ActorInfo))

(event Ability|EventActivateAbility
  (|CallParentFunction)
  (Class|FortControllerComponentTransientQuests|GrantTransientQuest (Class|FortPlayerPawn|GetTransientQuestsComponent (Variables|GAVars|Data|GetOwningPlayerPawn)) (Variables|Default|GetQuest%20Def%20Athena%20to%20Grant) 0)
  (Ability|EndAbility))

(event Ability|EventOnEndAbility
  (params (bWasCancelled :type Boolean)))
